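# Unlit translucent Niagara sprite master material for explosion streak particles.
# Opacity multiplies particle alpha by a streak texture, applies depth fade, then optionally blends in an adjustable-power near-camera fade.
# Particle color optionally blends toward a hue-shifted version driven by a dynamic parameter, then passes through a shared emissive adjustments function.
# The dynamic DepthFade channel is named but unused, as fade distance comes from a scalar parameter.
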Material /Game/Effects/Niagara/MasterMaterials/Spine/M_ExplosionStreak
graph {
 properties {
  bAllowNegativeEmissiveColor=true
  bUsedWithNanite=true
  bUsedWithNiagaraSprites=true
  blendMode=BLEND_Translucent
  pixelDepthOffsetMode=PDOM_Legacy
  shadingModel=MSM_Unlit
  usedShadingModels=Unlit
 }
 nodes [
  n1 = Power(
   Base=MaterialFunctionCall(Distance=ScalarParameter() {defaultValue=512, group=Opacity|Near_Camera_Fade, parameterName=NearCamFade_Distance, sortPriority=1}) {materialFunction=&Near_Camera_Fade}.Result
   Exp=ScalarParameter() {defaultValue=1, group=Opacity|Near_Camera_Fade, parameterName=NearCamFade_Power, sortPriority=2}
  )
  n2 = ParticleColor()
  n3 = DepthFade(
   Opacity=Multiply(
    A=@n2.A
    B=TextureSampleParameter2D() {parameterName=Default_Tex, texture=/Game/Effects/Niagara/MasterMaterials/Spine/Texture/T_Streak_Blur}.RGB
   )
   FadeDistance=ScalarParameter() {parameterName=FadeDistance}
  )
  LinearInterpolate(
   A=@n3
   B=Multiply(A=@n3, B=Clamp(None=@n1) {minDefault=0.02})
   Alpha=ScalarParameter() {group=Opacity|Near_Camera_Fade, parameterName=Add Near Camera Fade}
  )
  n4 = DynamicParameter() {defaultValue=<rgba 1 1 10 1>, paramNames [DepthFade,none,HueShift,none]}
  MaterialFunctionCall(Emissive=LinearInterpolate(
    A=@n2.RGB
    B=MaterialFunctionCall(Hue Shift Percentage=@n4.HueShift, Texture=@n2.RGB) {materialFunction=&HueShift}.Result
    Alpha=ScalarParameter() {group=Emissive_Color, parameterName="Use HUE && INTENSITY Shifts"}
   )) {materialFunction=&MF_FX_EmissiveAdjustments}
 ]
}
