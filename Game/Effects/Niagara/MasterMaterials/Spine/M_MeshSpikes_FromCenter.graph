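# Translucent, unlit, two-sided Niagara master material for spike-from-center effect meshes, also enabled for sprites.
# Particle color drives emissive through an adjustment function; a panning noise, weighted by inverted UV1 V, displaces vertices along normals.
# Opacity combines particle alpha, masked base texture, view-angle fade and depth fade, with optional height-lerp noise dissolve driven by a dynamic parameter.
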Material /Game/Effects/Niagara/MasterMaterials/Spine/M_MeshSpikes_FromCenter
graph {
 properties {
  bAutomaticallySetUsageInEditor=false
  bUsedWithNiagaraMeshParticles=true
  bUsedWithNiagaraSprites=true
  blendMode=BLEND_Translucent
  shadingModel=MSM_Unlit
  twoSided=true
  usedShadingModels=Unlit
 }
 nodes [
  n1 = ParticleColor()
  MaterialFunctionCall(Emissive=@n1.RGB) {materialFunction=&MF_FX_EmissiveAdjustments}
  n2 = TextureSample(UVs=Panner(Coordinate=TextureCoordinate() {coordinateIndex=1}) {speedY=2}) {samplerType=SAMPLERTYPE_Grayscale, texture=&LowResBlurredNoise}
  n3 = VertexNormalWS()
  n4 = TextureCoordinate() {coordinateIndex=1}
  n5 = Multiply(A=OneMinus(None=ComponentMask(None=@n4) {g=true})) {constB=5}
  Multiply(A=Multiply(A=Multiply(A=@n2.R, B=@n3), B=Saturate(None=@n5)), B=ScalarParameter() {defaultValue=10, parameterName=WPOIntensity}) {constB=10}
  n6 = TextureSampleParameter2D() {
   parameterName=Tex_Base
   samplerType=SAMPLERTYPE_LinearColor
   texture=/Game/Effects/Niagara/MasterMaterials/Spine/Texture/T_MeshSpikes_FromCenter
  }
  n7 = Multiply(
   A=Multiply(A=@n1.A, B=StaticComponentMaskParameter(None=@n6.RGB) {defaultR=true, parameterName=Mask})
   B=MaterialFunctionCall("FadeSensitivity (optional)"=ScalarParameter() {defaultValue=1, parameterName="FadeSensitivity (optional) (S)"}) {materialFunction=&FadeBasedOnViewAngle}.Result
  )
  n8 = LinearInterpolate(Alpha=DynamicParameter() {defaultValue=<rgba 1 1 1 0>, paramNames [Param1,Param2,Param3,Dissolve]}.Dissolve) {constA=0.3, constB=0.8}
  n9 = VectorParameter() {parameterName=DissUV, defaultValue=<rgba 1 1 0 0>}
  n10 = TextureSampleParameter2D(UVs=Multiply(A=TextureCoordinate(), B=ComponentMask(None=@n9.RGB) {g=true, r=true})) {
   parameterName=Tex_Dissolve
   samplerType=SAMPLERTYPE_LinearGrayscale
   texture=&T_SimplePerlin_Greyscale
  }
  n11 = ScalarParameter() {defaultValue=3, parameterName="Contrast (S)"}
  Multiply(
   A=DepthFade(Opacity=Multiply(A=@n7), FadeDistance=ScalarParameter() {defaultValue=5, parameterName=FadeDistance}) {fadeDistanceDefault=5}
   B=LinearInterpolate(
    B=MaterialFunctionCall(Transition Phase=@n8, Height Texture=@n10.R, Contrast=@n11) {materialFunction=&HeightLerp}.Results
    Alpha=ScalarParameter() {parameterName=UseDissolve}
   ) {constA=1}
  )
 ]
}
