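# Translucent unlit Niagara material for sprite and mesh particles, apparently a red-to-yellow sprint streak effect.
# A randomly chosen flipbook frame and a vertical gradient blend two colors, multiplied by a streak texture and particle color into emissive adjustments.
# Opacity combines both shapes with randomly offset panning noise and caustics, particle alpha, an optional linear mask and depth fade.
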Material /Game/Effects/Niagara/MasterMaterials/Spine/M_Red_SprintSpine
graph {
 properties {
  bUsedWithNiagaraMeshParticles=true
  bUsedWithNiagaraSprites=true
  blendMode=BLEND_Translucent
  pixelDepthOffsetMode=PDOM_Legacy
  shadingModel=MSM_Unlit
  usedShadingModels=Unlit
 }
 nodes [
  MaterialFunctionCall() {materialFunction=&FadeBasedOnViewAngle}
  n1 = Power(
   Base=MaterialFunctionCall() {materialFunction=&LinearGradient}.VGradient
   Exp=ScalarParameter() {defaultValue=0.2, parameterName="HueShift(RedToYellow)"}
  ) {constExponent=0.3}
  n2 = ParticleRandom()
  n3 = Constant() {r=1}
  n4 = ScalarParameter() {defaultValue=2, parameterName=NumberOfColumns}
  n5 = TextureSampleParameter2D(UVs=Add(A=MaterialFunctionCall("Animation  Phase (0-1)"=@n2, Number of Rows=@n3, Number of Columns=@n4) {materialFunction=&FlipBook}.UVs) {constB=0}) {parameterName=Texture, texture=/Game/Weapons/FORT_Pistols/Effects/Muzzle/T_Muzzle_Flash_01_Sharp}
  n6 = Multiply(A=@n5.R) {constB=2}
  n7 = LinearInterpolate(
   A=VectorParameter() {parameterName=Color01, defaultValue=<rgba 1 0.7635 0.0915 1>}.RGB
   B=VectorParameter() {parameterName=Color02, defaultValue=<rgba 1 0.0563 0 1>}.RGB
   Alpha=Saturate(None=LinearInterpolate(A=@n1, B=@n6))
  )
  n8 = TextureSampleParameter2D(UVs=MaterialFunctionCall(
    "Animation  Phase (0-1)"=Constant() {r=0.75}
    Number of Rows=Constant() {r=1}
    Number of Columns=ScalarParameter() {defaultValue=4, parameterName=Texture02_UV_number}
   ) {materialFunction=&FlipBook}.UVs) {parameterName=Texture02, texture=/Game/Effects/Niagara/MasterMaterials/Spine/Texture/T_Streak_Blur}
  n9 = ParticleColor()
  MaterialFunctionCall(Emissive=Multiply(A=Multiply(A=@n7, B=@n8.R), B=@n9.RGB)) {materialFunction=&MF_FX_EmissiveAdjustments}
  n10 = TextureCoordinate() {vTiling=2}
  n11 = Constant2Vector() {g=0.1, r=0.5}
  n12 = Multiply(A=Time(), B=ScalarParameter() {defaultValue=1, parameterName=Speed})
  n13 = AppendVector(A=ScalarParameter() {defaultValue=0.5, parameterName=OpacitySpeedX}, B=ScalarParameter() {defaultValue=1, parameterName=OpacitySpeedY})
  n14 = Panner(Coordinate=Add(A=Multiply(A=@n10, B=@n11), B=ParticleRandom()), Time=@n12, Speed=@n13) {speedX=0.5, speedY=1}
  n15 = TextureCoordinate()
  n16 = AppendVector(
   A=ScalarParameter() {defaultValue=1, parameterName=OpacityTextureB_U}
   B=ScalarParameter() {defaultValue=0.2, parameterName=OpacityTextureB_V}
  )
  n17 = LinearInterpolate(
   A=Add(A=TextureCoordinate() {vTiling=0.2}, B=ParticleRandom())
   B=Add(A=Multiply(A=@n15, B=@n16), B=ParticleRandom())
   Alpha=ScalarParameter() {parameterName=Use_OpacityTexB_UV}
  )
  n18 = Multiply(A=TextureSampleParameter2D(UVs=Panner(Coordinate=@n17, Time=@n12, Speed=@n13) {speedX=0.5, speedY=1}) {parameterName=OpacityTextureB, texture=&T_ShieldCaustics}.R)
  n19 = Multiply(
   A=Add(A=@n5.R, B=@n8.R)
   B=Multiply(
    A=TextureSample(UVs=@n14) {samplerType=SAMPLERTYPE_Grayscale, texture=&LowResBlurredNoise}.R
    B=Power(Base=@n18) {constExponent=0.6}
   )
  )
  n20 = Multiply(A=Multiply(A=@n9.A, B=Saturate(None=@n19)))
  n21 = Multiply(
   A=@n20
   B=TextureSampleParameter2D() {parameterName=LinearMaskTexture, texture=/Game/PvP/StrategicBuildingActors/Defensive/LaserFence/Assets/T_LaserBeamParticle}.R
  )
  n22 = ScalarParameter() {parameterName=Use_LinearMaskTexture}
  Multiply(A=DepthFade(Opacity=LinearInterpolate(A=@n20, B=@n21, Alpha=@n22), FadeDistance=ScalarParameter() {defaultValue=20, parameterName=DepthFade}) {fadeDistanceDefault=20})
  n23 = Multiply(A=TextureCoordinate(), B=Constant2Vector() {g=1, r=0.4})
  n24 = ParticleRandom()
  n25 = TextureSample(UVs=Panner(Coordinate=Add(A=@n23, B=@n24)) {speedX=0.4, speedY=1}) {texture=/Game/Effects/Fort_Effects/Emotes/UltraLight/Texture/T_SoftNoise_BC7}
  Multiply(A=AppendVector(A=@n25.R, B=@n25.R)) {constB=0.05}
  ParticleRandom()
  ScalarParameter() {defaultValue=1, parameterName=FadeBasedOnViewAngle}
  TextureCoordinate()
 ]
}
